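# Niagara emote effect that surrounds the player with a miniature city of mesh buildings that follows their movement.
# One emitter spawns a flat grid, split into small, medium and tall buildings with varied quarter-turn rotations.
# Particles treadmill around the owner and trace downward for ground height, using socket height at high speed or in front-end preview, which also scrolls them.
# A radial mask drives an overshooting scale timeline so buildings pop up and shrink away.
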NiagaraSystem /Game/Effects/Niagara/Systems/CosmeticTemplates/Emote/City/NS_City
dependencies {
 usedDataInterfaces [
  &NiagaraDataInterfaceCurve
  &NiagaraDataInterfaceMeshRendererInfo
  &NiagaraDataInterfaceArrayFloat3
  &NiagaraDataInterfaceCollisionQuery
  &NiagaraDataInterfaceDebugDraw
  &NiagaraDataInterfaceArrayFloat
  &NiagaraDataInterfaceSocketReader
 ]
 usedDynamicInputs [
  &MakeIntFromFloat
  &RandomRangeFloat
  &FNTimeOfDayColor
 ]
 usedModules [
  &self:SystemSpawnScript.SystemScriptSource.SystemScriptGraph.NiagaraNodeAssignment_5.SetVariables_#1
  &SystemState
  &self:SystemSpawnScript.SystemScriptSource.SystemScriptGraph.NiagaraNodeAssignment_6.SetVariables_#2
  &self:SystemSpawnScript.SystemScriptSource.SystemScriptGraph.NiagaraNodeAssignment_1.SetVariables_#3
 ]
}
emitters {CityTemplate {
  emitterUpdateScript [
   &EmitterState(
    Inactive Response=::NewEnumerator1
    Life Cycle Mode=::NewEnumerator1
    Loop Behavior=::NewEnumerator1
    Loop Duration Mode=::NewEnumerator1
    UseLoopDelay=false
   )
   /SetVariables(
    Emitter.OriginOffset:Vector3f=&TransformVector(Vector=<0 0 0>, Destination Space=::World)
    Emitter.PaddingPerCell=6
    Emitter.VolumeRadius=300
   ) {isSetParameters=true}
   /Niagara/Modules/Spawn/Location/SpawnParticlesInGrid(X Count=12, Y Count=12, Z Count=0)
  ]
  particleSpawnScript [
   &PartitionParticles(Number Of Partitions=3, Partition Mode:ENiagara_PartitionMode=::NewEnumerator5) {module=PartitionParticles001}
   &self:PartitionByBuildingSize_02(
    Fraction_MediumB=0.35
    Franction_TallB=0.2
    LBuilding_IndexMax=10
    MBuilding_IndexMax=20
    MBuilding_IndexMin=11
    SBuilding_IndexMax=25
    SBuilding_IndexMin=21
   ) {module=PartitionByBuildingSize_02}
   &FN_SetSpawnGroup(Number of Spawn Groups=6) {module=FN_SetSpawnGroup001}
   &InitializeParticle(
    Mesh Scale Mode=::NewEnumerator3
    Mesh Uniform Scale=0.6
    Position Offset Coordinate Space=::Local
    UsePositionOffset=true
    Write Material Random=true
   )
   /Niagara/Modules/Spawn/Location/V2/GridLocation(
    Coordinate Space:ENiagaraCoordinateSpace=::World
    Module.Jitter Inside Grid Cells=false
    Module.Override Random Seed=false
    Normalize Offsets=false
    XYZ Dimensions:Vector3f=&MakeVector(
     X=&Divide_Float(A=$Emitter.VolumeRadius, B=$Emitter.PaddingPerCell)
     Y=&Divide_Float(A=$Emitter.VolumeRadius, B=$Emitter.PaddingPerCell)
     Z=0
    )
   )
   &InitialMeshOrientation(
    Mesh Orientation Mode=::NewEnumerator2
    Rotation=&SelectVectorFromArray(
     Array Sampling Mode=::NewEnumerator0
     Direct Array Index=?
     Vector Selection Array=DataInterface() {propertyValues {FloatData [<0 0 0>, <0 0 0.25>, <0 0 0.5>]}}
     Random Seed=0
    )
    Use Rotation Vector=true
   )
   /SetVariables() {isSetParameters=true}
  ]
  particleUpdateScript [
   &ParticleState(Kill Particles When Lifetime Has Elapsed=false, Let Infinitely Lived Particles Die When Emitter Deactivates=true)
   &SocketLocation(
    Orientation Sampling=::NewEnumerator1
    Position Sampling=::NewEnumerator1
    SamplePosOffset=false
    Socket Reader=$System.Socket Reader
    Socket Sampling Mode=::NewEnumerator3
    VelocitySpeedLimit=false
   )
   &SolveForcesAndVelocity(Clamp Velocity=false, Limit Acceleration=false)
   /SetVariables(Particles.PositionCenter:NiagaraPosition=&AddVectorToPosition(Position=$Engine.Owner.Position, Vector=$Emitter.OriginOffset)) {isSetParameters=true}
   &self:TreadmillParticles(
    GridSize:NiagaraFloat=&Multiply_Float(A=$Emitter.VolumeRadius, B=2)
    InputPosition:NiagaraPosition=$Particles.PositionCenter
   ) {module=TreadmillParticles}
   &DoOnce(
    Reset:NiagaraBool=&SetBoolByFloatComparison(A=$Particles.Timeline001.Time, B=0.9)
    Trigger Condition:NiagaraBool=&SetBoolByFloatComparison(
     A=&VectorLength(Vector=Linked() {linkedVariable {name=Particles.Position, type=NiagaraPosition}})
     B=&VectorLength(Vector=Linked() {linkedVariable {name=Particles.Previous.Position, type=NiagaraPosition}})
     Comparison Type=::NewEnumerator5
    )
   )
   &Timeline(Instantaneous Rewind=true, Loop While Playing=true, Play=$Particles.DoOnce.TriggeredOnce, Play Rate=10) {module=Timeline001}
   &self:GridCollision_01(
    CharacterPos:NiagaraPosition=&AddVectorToPosition(Position=$Engine.Owner.Position, Vector=<0 0 50>)
    Disable Trace:NiagaraBool=&Boolean_AndOperation(
     Bool1=&Boolean_NotOperation(Bool=$Particles.DoOnce.TriggeredOnce)
     Bool2=&SetBoolByFloatComparison(A=$Emitter.Age, B=1)
    )
    Height Adjustment=1.5
    Trace Origin Height=150
    TraceDir:Vector3f=<0 0 -1>
    TraceDist:NiagaraFloat=$Emitter.VolumeRadius
    bIsFrontEnd:NiagaraBool=$User.bIsFrontEndPreview
   ) {module=GridCollision_01}
   /SetVariables(
    Particles.Position=/Niagara/DynamicInputs/TypeConversions/MakeCustomPositionFromBool(
     Bool:NiagaraBool=&Boolean_OrOperation(
      Bool1=$User.bIsFrontEndPreview
      Bool2=&SetBoolByFloatComparison(A=&VectorLength(Vector=$Engine.Owner.Velocity), B=300)
     )
     True Position:NiagaraPosition=&AddVectorToPosition(
      Position=&MakeVector(
       X=&MakeFloatFromVector(Vector=Linked() {linkedVariable {name=Particles.Position, type=NiagaraPosition}}, Channel=::NewEnumerator0)
       Y=&MakeFloatFromVector(Vector=Linked() {linkedVariable {name=Particles.Position, type=NiagaraPosition}}, Channel=::NewEnumerator1)
       Z=&MakeFloatFromVector(Vector=Linked() {linkedVariable {name=Output.SocketLocation.SampledPosition, type=NiagaraPosition}}, Channel=::NewEnumerator2)
      )
      Vector=&FN_FrontEndVector3Modifier(
       Gameplay=<0 0 0>
       Front End=&RotateVectorByQuaternion(
        Vector To Rotate=&Multiply_VectorByFloat(Vector=<0 -50 0>, Float=$Engine.DeltaTime)
        Quaternion=$Output.SocketLocation.SampledRotation
       )
      )
     )
     False Position:NiagaraPosition=&MakeCustomPositionFromBool(
      Bool:NiagaraBool=$Particles.DoOnce.TriggeredOnce
      True Position:NiagaraPosition=$Particles.CollisionPos
      False Position:NiagaraPosition=&MakeCustomPositionFromBool(
       Bool:NiagaraBool=&SetBoolByFloatComparison(A=$Emitter.Age, B=1)
       True Position:NiagaraPosition=$Particles.Previous.Position
       False Position:NiagaraPosition=$Particles.CollisionPos
      )
     )
    )
    Particles.SpriteFacing=$Particles.CollisionNormal
   ) {isSetParameters=true}
   &RegionMask(
    Invert Region=true
    Offset:Vector3f=$Emitter.OriginOffset
    Region Falloff=0
    Region Radius:NiagaraFloat=&Add_Float(
     A=&Multiply_Float(A=$Emitter.VolumeRadius, B=1.5)
     B=&Multiply_Float(A=$Emitter.VolumeRadius, B=-0.7)
    )
    UseOffset=true
    UseRotation=false
   )
   &Timeline(
    Additional Output:ENiagara_TimelineOutput=::NewEnumerator1
    Additional Output Mode:ENiagara_TimelineOutputMode=::NewEnumerator1
    Float Curve=DataInterface() {propertyValues {Curve {keys [
        {arriveTangent=0.9218, interpMode=RCIM_Cubic, leaveTangent=0.9218, tangentMode=RCTM_User, time=0.35, value=0}
        {arriveTangent=2.875, interpMode=RCIM_Cubic, leaveTangent=2.875, time=0.63, value=0.9}
        {arriveTangent=-0.0156, interpMode=RCIM_Cubic, leaveTangent=-0.0156, tangentMode=RCTM_User, time=0.75, value=1.15}
        {arriveTangent=-0.0297, interpMode=RCIM_Cubic, leaveTangent=-0.0297, tangentMode=RCTM_User, time=0.9, value=0.95}
        {arriveTangent=0.0389, interpMode=RCIM_Cubic, leaveTangent=0.0389, tangentMode=RCTM_User, time=0.96, value=1.025}
        <at 1 1 RCIM_Cubic>
       ]}}}
    Play=&Boolean_NotOperation(Bool=$Particles.RegionMask.IsInsideMask)
    Play Rate=&MakeCustomFloatFromBool(
     Bool=&SetBoolByFloatComparison(A=&VectorLength(Vector=$User.OwnerLocalVelocity), B=160)
     True Float=10
     False Float=1
    )
    Rewind Rate:NiagaraFloat=&MakeCustomFloatFromBool(
     Bool=&SetBoolByFloatComparison(A=&VectorLength(Vector=$User.OwnerLocalVelocity), B=160)
     True Float=10
     False Float=3
    )
   )
   &ScaleMeshSize(Scale Factor=&VectorFromFloat(Value=&Multiply_Float(
      A=$Particles.Timeline.Float
      B=&FloatFromCurve(FloatCurve=DataInterface() {propertyValues {Curve {
          defaultValue=1
          keys [{interpMode=RCIM_Cubic, tangentMode=RCTM_User, time=0.3101, value=0}, {interpMode=RCIM_Cubic, tangentMode=RCTM_User, time=1, value=1}]
         }}})
     ))) {module=ScaleMeshSize001}
  ]
  rendererClasses [&NiagaraMeshRendererProperties]
  renderers [{rendererClass=&NiagaraMeshRendererProperties, rendererIndex=0}]
 }}
summary {
 systemName=NS_City
 emitters [{emitterName=CityTemplate, rendererClasses [&NiagaraMeshRendererProperties]}]
 userVariables [
  User.bIsFrontEnd {type=NiagaraBool}
  User.bIsFrontEndPreview {type=NiagaraBool}
  User.OwnerLocalVelocity {type=Vector3f, description="", defaultValue {struct=Vector3f, value=<1 1 1>}}
 ]
}
